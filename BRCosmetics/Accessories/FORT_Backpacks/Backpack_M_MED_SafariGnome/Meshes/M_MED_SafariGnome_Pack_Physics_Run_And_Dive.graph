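# Physics asset for the SafariGnome backpack cosmetic, named for run and dive, that sets up a dangling decoration with a tassel.
# decor_dyn_all hinges off decor_dyn_MSTR on one swing axis and carries mirrored front and back chains of six small capsules with tight swing limits.
# The chain ends are locked together into a closed loop, with the tassel sphere jointed to the back end.
# A separate SwordHide box has no constraints.
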PhysicsAsset /BRCosmetics/Accessories/FORT_Backpacks/Backpack_M_MED_SafariGnome/Meshes/M_MED_SafariGnome_Pack_Physics_Run_And_Dive
bodies {
 SwordHide [{shapeType=Box, extentX=11.9081, extentY=1.35, extentZ=4, shapeName=SwordHide_box, center=<21.2565 0 0>}]
 decor_ba_dyn_01 [{shapeType=Capsule, length=0.86, radius=0.2, shapeName=decor_ba_dyn_01_capsule, center=<0.43 0 0>, rotation=<rot 0 -90 -90>}]
 decor_ba_dyn_02 [{shapeType=Capsule, length=0.979, radius=0.2, shapeName=decor_ba_dyn_02_capsule, center=<0.4895 0 0>, rotation=<rot 0 -90 -90>}]
 decor_ba_dyn_03 [{shapeType=Capsule, length=1.633, radius=0.2, shapeName=decor_ba_dyn_03_capsule, center=<0.8165 0 0>, rotation=<rot 0 -90 -90>}]
 decor_ba_dyn_04 [{shapeType=Capsule, length=1.391, radius=0.2, shapeName=decor_ba_dyn_04_capsule, center=<0.6955 0 0>, rotation=<rot 0 -90 -90>}]
 decor_ba_dyn_05 [{shapeType=Capsule, length=0.579, radius=0.2, shapeName=decor_ba_dyn_05_capsule, center=<0.2895 0 0>, rotation=<rot 0 -90 -90>}]
 decor_ba_dyn_06 [{shapeType=Capsule, length=0.01, radius=0.2, shapeName=decor_ba_dyn_06_capsule, rotation=<rot 0 -90 -90>}]
 decor_ct_tassle [{shapeType=Sphere, radius=2.625, shapeName=decor_ct_tassle_sphere, center=<2.6922 0 -0.0006>}]
 decor_dyn_MSTR [{shapeType=Sphere, radius=0.125, shapeName=decor_dyn_MSTR_sphere}]
 decor_dyn_all [{shapeType=Capsule, length=0.505, radius=0.05, shapeName=decor_dyn_all_capsule, center=<0.1498 0 0>, rotation=<rot 0 180 -90>}]
 decor_fr_dyn_01 [{shapeType=Capsule, length=0.86, radius=0.2, shapeName=decor_fr_dyn_01_capsule, center=<0.43 0 0>, rotation=<rot 0 -90 -90>}]
 decor_fr_dyn_02 [{shapeType=Capsule, length=0.979, radius=0.2, shapeName=decor_fr_dyn_02_capsule, center=<0.4895 0 0>, rotation=<rot 0 -90 -90>}]
 decor_fr_dyn_03 [{shapeType=Capsule, length=1.633, radius=0.2, shapeName=decor_fr_dyn_03_capsule, center=<0.8165 0 0>, rotation=<rot 0 -90 -90>}]
 decor_fr_dyn_04 [{shapeType=Capsule, length=1.391, radius=0.2, shapeName=decor_fr_dyn_04_capsule, center=<0.6955 0 0>, rotation=<rot 0 -90 -90>}]
 decor_fr_dyn_05 [{shapeType=Capsule, length=0.579, radius=0.2, shapeName=decor_fr_dyn_05_capsule, center=<0.2895 0 0>, rotation=<rot 0 -90 -90>}]
 decor_fr_dyn_06 [{shapeType=Capsule, length=0.01, radius=0.2, shapeName=decor_fr_dyn_06_capsule, rotation=<rot 0 -90 -90>}]
}
constraints [
 {
  bone1Name=decor_dyn_all
  bone2Name=decor_dyn_MSTR
  swing1LimitDegrees=45
  swing1Motion=Locked
  swing2LimitDegrees=45
  twistLimitDegrees=45
  twistMotion=Locked
 }
 {bone1Name=decor_ba_dyn_01, bone2Name=decor_dyn_all, swing1LimitDegrees=10, swing2LimitDegrees=10, twistLimitDegrees=45, twistMotion=Locked}
 {bone1Name=decor_ba_dyn_02, bone2Name=decor_ba_dyn_01, swing1LimitDegrees=10, swing2LimitDegrees=10, twistLimitDegrees=45, twistMotion=Locked}
 {bone1Name=decor_ba_dyn_03, bone2Name=decor_ba_dyn_02, swing1LimitDegrees=10, swing2LimitDegrees=5, twistLimitDegrees=45, twistMotion=Locked}
 {bone1Name=decor_ba_dyn_04, bone2Name=decor_ba_dyn_03, swing1LimitDegrees=10, swing2LimitDegrees=5, twistLimitDegrees=45, twistMotion=Locked}
 {bone1Name=decor_ba_dyn_05, bone2Name=decor_ba_dyn_04, swing1LimitDegrees=10, swing2LimitDegrees=5, twistLimitDegrees=45, twistMotion=Locked}
 {bone1Name=decor_ba_dyn_06, bone2Name=decor_ba_dyn_05, swing1LimitDegrees=10, swing2LimitDegrees=10, twistLimitDegrees=5}
 {bone1Name=decor_fr_dyn_01, bone2Name=decor_dyn_all, swing1LimitDegrees=10, swing2LimitDegrees=10, twistLimitDegrees=45, twistMotion=Locked}
 {bone1Name=decor_fr_dyn_02, bone2Name=decor_fr_dyn_01, swing1LimitDegrees=10, swing2LimitDegrees=10, twistLimitDegrees=45, twistMotion=Locked}
 {bone1Name=decor_fr_dyn_03, bone2Name=decor_fr_dyn_02, swing1LimitDegrees=10, swing2LimitDegrees=5, twistLimitDegrees=45, twistMotion=Locked}
 {bone1Name=decor_fr_dyn_04, bone2Name=decor_fr_dyn_03, swing1LimitDegrees=10, swing2LimitDegrees=5, twistLimitDegrees=45, twistMotion=Locked}
 {bone1Name=decor_fr_dyn_05, bone2Name=decor_fr_dyn_04, swing1LimitDegrees=10, swing2LimitDegrees=5, twistLimitDegrees=45, twistMotion=Locked}
 {bone1Name=decor_fr_dyn_06, bone2Name=decor_fr_dyn_05, swing1LimitDegrees=10, swing2LimitDegrees=10, twistLimitDegrees=5}
 {
  bone1Name=decor_ba_dyn_06
  bone2Name=decor_fr_dyn_06
  swing1LimitDegrees=10
  swing1Motion=Locked
  swing2LimitDegrees=10
  swing2Motion=Locked
  twistLimitDegrees=45
  twistMotion=Locked
 }
 {bone1Name=decor_ct_tassle, bone2Name=decor_ba_dyn_06, swing1LimitDegrees=15, swing2LimitDegrees=15, twistLimitDegrees=5}
]
