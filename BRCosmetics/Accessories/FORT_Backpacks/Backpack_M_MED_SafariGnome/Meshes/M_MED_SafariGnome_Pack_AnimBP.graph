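# Animation Blueprint for the Safari Gnome backpack that adjusts pack placement and simulates its dangling decor.
# Each update maps spine bend between spine_01 and spine_05 to additive pack rotation and translation.
# Two RigidBody branches share a physics asset: slow skydive descent uses upward gravity with rotated decor, otherwise normal downward gravity applies.
# SkyNormalDescent, true when skydiving and falling without diving or an open parachute, smoothly blends between them.
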AnimBlueprint /BRCosmetics/Accessories/FORT_Backpacks/Backpack_M_MED_SafariGnome/Meshes/M_MED_SafariGnome_Pack_AnimBP
target_skeleton=&/M_MED_SafariGnome_Pack_Skeleton
graphs [
 AnimGraph {nodes [
   n1 = "Transform (Modify) Bone - Bone: pack"(
    ComponentPose=Local To Component(LocalPose=Use cached pose 'Presim'())
    Translation=Get Backpack Translate() -> (Backpack Translate:Vector)
    Rotation=Get Backpack Rotate() -> (Backpack Rotate:Rotator)
   ) {rotationMode=BMM_Additive, translationMode=BMM_Additive, alphaBoolBlend {blendOption=Linear}, boneToModify {boneName=pack}}
   n2 = "Transform (Modify) Bone - Bone: decor_dyn_MSTR"(ComponentPose=@n1, Alpha=1, Translation="0,0,0", Rotation="-120,0,0", Scale="1,1,1") {
    rotationMode=BMM_Replace
    rotationSpace=BCS_BoneSpace
    scaleSpace=BCS_BoneSpace
    translationSpace=BCS_BoneSpace
    alphaBoolBlend {blendOption=Linear}
    boneToModify {boneName=decor_dyn_MSTR}
    rotation {pitch=-120}
   }
   n3 = RigidBody(ComponentPose=@n2) {
    bOverrideWorldGravity=true
    overridePhysicsAsset=&/M_MED_SafariGnome_Pack_Physics_Run_And_Dive
    simulationSpace=BaseBoneSpace
    alphaBoolBlend {blendOption=Linear}
    baseBoneRef {boneName=root}
    componentAppliedLinearAccClamp=<1500 1500 1500>
    componentLinearAccScale=<0.5 0.5 0.5>
    componentLinearVelScale=<0.5 0.5 0.5>
    overrideWorldGravity {z=320}
   }
   n4 = "Transform (Modify) Bone - Bone: pack"(ComponentPose=Local To Component(LocalPose=Use cached pose 'Presim'()), Translation=Get Backpack Translate(), Rotation=Get Backpack Rotate()) {
    rotationMode=BMM_Additive
    rotationSpace=BCS_BoneSpace
    scaleSpace=BCS_BoneSpace
    translationMode=BMM_Additive
    translationSpace=BCS_BoneSpace
    alphaBoolBlend {blendOption=Linear}
    boneToModify {boneName=pack}
   }
   n5 = RigidBody(ComponentPose=@n4) {
    bOverrideWorldGravity=true
    overridePhysicsAsset=&M_MED_SafariGnome_Pack_Physics_Run_And_Dive
    simulationSpace=BaseBoneSpace
    alphaBoolBlend {blendOption=Linear}
    baseBoneRef {boneName=root}
    componentAppliedLinearAccClamp=<1500 1500 1500>
    componentLinearAccScale=<0.5 0.5 0.5>
    componentLinearVelScale=<0.5 0.5 0.5>
    overrideWorldGravity {z=-980}
   }
   Output Pose(Result=Blend Poses by bool(
     BlendPose_0=Component To Local(ComponentPose=@n3)
     BlendPose_1=Component To Local(ComponentPose=@n5)
     BlendTime_0=0.25
     BlendTime_1=0.3
     bActiveValue=Get SkyNormalDescent() -> (SkyNormalDescent:bool)
    ) {blendType=CubicInOut}) {layerGroup=None}
   Save cached pose 'Presim'(Pose=Input Pose()) {cacheName=Presim}
  ]}
 EventGraph {nodes [
   Event BlueprintUpdateAnimation()
   n1 = GetOwnerSkeletalMeshComponent()
   n2 = BreakRotator(InRot=BreakTransform(InTransform=GetDeltaTransformFromRefPose(self=@n1, BoneName=spine_05, BaseName=spine_01)).Rotation)
   n3 = MapRangeClamped(Value=@n2.Yaw, InRangeA=0, InRangeB=-30, OutRangeB=35)
   > Set Backpack Rotate(Backpack Rotate:Rotator=MakeRotator(Roll=float * float(A=@n3, B=1.333) {is=title K2Node_PromotableOperator})) -> (Output_Get:Rotator)
   n4 = MapRangeClamped(Value=@n2.Yaw, InRangeA=0, InRangeB=-30, OutRangeB=-10)
   > Set Backpack Translate(Backpack Translate:Vector=vector + vector(A=MakeVector(Y=@n4), B="0,-2.25,0")) -> (Output_Get:Vector)
   n5 = Get bIsSkydiveDiveMode()
   n6 = Get bIsParachuteOpened()
   > n7 = Branch(Condition=AND Boolean!(A=Get bIsSkyDiving(), B=Get bIsFalling(), C:bool=NOT Boolean(A=@n6), D:bool=NOT Boolean(A=@n5)) -> (ReturnValue))
   > Set SkyNormalDescent(SkyNormalDescent:bool=true) -> (Output_Get:bool)
   > PrintString!(execute, InString="Falling slowly!", bPrintToScreen, bPrintToLog, TextColor, Duration, Key) -> (then)
   Set SkyNormalDescent(execute=@n7.else, SkyNormalDescent=false)
   TryGetPawnOwner()
  ]}
]
variables [
 Backpack Rotate {type=Rotator}
 Backpack Translate {type=Vector}
 SkyNormalDescent {type=boolean}
]
